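# An editor material for Cascade, named for its curve editor.
# Its graph holds a single texture sample of the CASC_CurveEd texture, with no connection to a material output shown.
# It uses the Default Lit shading model and legacy pixel depth offset, with no math or parameters.
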Material /Engine/EditorMaterials/Cascade/CASC_CurveEd_MAT
graph {
 properties {pixelDepthOffsetMode=PDOM_Legacy, usedShadingModels=Default Lit}
 nodes [TextureSample() {texture=&/CASC_CurveEd}]
}
